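# Sound cue for a grappler item drop in the Figment Athena Dynamite content, judging by its name and weapon-drop settings.
# A single wave player feeds the root directly, reusing a dual pistols drop wave with no randomization or mixing.
# It plays at reduced volume with shared weapon-drop attenuation, the loot-collect sound class, and a stop-oldest weapon-drop concurrency set.
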SoundCue /Figment_Athena_Dynamite/Audio/Grappling_Drop_Figment_Cue
root=WavePlayer_1
nodes [{id=WavePlayer_1, properties {soundWaveAssetPtr=/Figment_Athena_Dynamite/Audio/Wavs/DualPistols_Drop_02}}]
properties {
 attenuationSettings=&Fort_Weapon_Drop_Att
 maxDistance=3000
 soundClassObject=&Fort_Loot_Collect
 volumeMultiplier=0.4
 concurrencyOverrides {maxCount=1, resolutionRule=PreventNew}
 concurrencySet [&Weapon_Drop_StopOldest]
}
