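# Sound cue for picking up the Figment shockwave grenade, part of the Figment Athena Dynamite content.
# A single wave player plays the existing shockwave grenade pickup sound at reduced volume and raised pitch.
# It uses the weapon pickup attenuation and the gadgets and items sound class, with a concurrency override that prevents new instances when the limit is reached.
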SoundCue /Figment_Athena_Dynamite/Audio/SC_Grenade_Shockwave_Figment_Pickup
root=WavePlayer_1
nodes [{id=WavePlayer_1, properties {soundWaveAssetPtr=/Game/Sounds/Fort_GamePlay_Sounds/Explosives/ShockwaveGrenade/Grenade_Shockwave_Pickup01}}]
properties {
 attenuationSettings=&Fort_Weapon_Pickup_Att
 bOverrideConcurrency=true
 maxDistance=3500
 pitchMultiplier=1.2
 soundClassObject=&Fort_Gadgets_Items
 volumeMultiplier=0.5
 concurrencyOverrides {resolutionRule=PreventNew}
}
