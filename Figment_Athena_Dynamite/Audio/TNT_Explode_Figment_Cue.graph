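# Sound cue for the Figment dynamite's TNT explosion, mixing close and distant layers with separate sound classes and attenuation.
# The first quality level uses three random variants per layer plus a pitched-down Dragon Shotgun impact-fuse sweetener in the close layer.
# The other three quality levels share a lighter mix with two variants per layer and no sweetener.
# Modulators vary pitch, distant layers add low-pass filtering and listener focus, and concurrency caps simultaneous plays.
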SoundCue /Figment_Athena_Dynamite/Audio/TNT_Explode_Figment_Cue
root=QualityLevel_0
nodes [
 {
  id=Attenuation_0
  children [Modulator_1]
  properties {
   bOverrideAttenuation=true
   attenuationOverrides {
    bAttenuateWithLPF=true
    bEnableListenerFocus=true
    distanceAlgorithm=NaturalSound
    falloffDistance=5500
    lPFFrequencyAtMax=200
    lPFRadiusMax=2000
    lPFRadiusMin=1000
    nonFocusAzimuth=45
    nonFocusVolumeAttenuation=0.3
    stereoSpread=3500
    attenuationShapeExtents {x=3000}
   }
  }
 }
 {
  id=Attenuation_2
  children [Modulator_0]
  properties {
   bOverrideAttenuation=true
   attenuationOverrides {distanceAlgorithm=NaturalSound, falloffDistance=3000, stereoSpread=8000, attenuationShapeExtents {x=1100}}
  }
 }
 {
  id=Attenuation_3
  children [Modulator_4]
  properties {
   bOverrideAttenuation=true
   attenuationOverrides {
    bAttenuateWithLPF=true
    bEnableListenerFocus=true
    distanceAlgorithm=NaturalSound
    falloffDistance=5500
    lPFFrequencyAtMax=200
    lPFRadiusMax=2000
    lPFRadiusMin=1000
    nonFocusAzimuth=45
    nonFocusVolumeAttenuation=0.3
    stereoSpread=3500
    attenuationShapeExtents {x=3000}
   }
  }
 }
 {
  id=Attenuation_6
  children [Modulator_3]
  properties {
   bOverrideAttenuation=true
   attenuationOverrides {distanceAlgorithm=NaturalSound, falloffDistance=3000, stereoSpread=8000, attenuationShapeExtents {x=1100}}
  }
 }
 {id=Mixer_0, children [Random_0,Modulator_5], properties {inputVolume [1.2,0.4]}}
 {id=Mixer_1, children [SoundClass_1,SoundClass_4], properties {inputVolume [1,1]}}
 {id=Mixer_4, children [SoundClass_3,SoundClass_0], properties {inputVolume [1.1,0.7]}}
 {id=Modulator_0, children [Mixer_0], properties {pitchMax=1.1, volumeMax=1, volumeMin=1}}
 {id=Modulator_1, children [Random_2], properties {pitchMax=1.2, pitchMin=0.85}}
 {id=Modulator_3, children [Random_3], properties {pitchMax=1.1, volumeMax=1, volumeMin=1}}
 {id=Modulator_4, children [Random_5], properties {pitchMax=1.2, pitchMin=0.85}}
 {id=Modulator_5, children [Random_1], properties {pitchMax=0.8, pitchMin=0.7}}
 {id=QualityLevel_0, children [Mixer_4,Mixer_1,Mixer_1,Mixer_1]}
 {id=Random_0, children [WavePlayer_3,WavePlayer_5,WavePlayer_6], properties {bShouldExcludeFromBranchCulling=true, weights [1,1,1]}}
 {id=Random_1, children [WavePlayer_7,WavePlayer_8,WavePlayer_9], properties {bShouldExcludeFromBranchCulling=true, weights [1,1,1]}}
 {id=Random_2, children [WavePlayer_15, WavePlayer_16, WavePlayer_17], properties {bShouldExcludeFromBranchCulling=true, weights [1,1,1]}}
 {id=Random_3, children [WavePlayer_18, WavePlayer_19], properties {bShouldExcludeFromBranchCulling=true, weights [1,1]}}
 {id=Random_5, children [WavePlayer_24, WavePlayer_25], properties {bShouldExcludeFromBranchCulling=true, weights [1,1]}}
 {
  id=SoundClass_0
  children [Attenuation_0]
  properties {soundClassOverride=&Fort_WeaponRanged_Fire_Distant}
 }
 {
  id=SoundClass_1
  children [Attenuation_6]
  properties {soundClassOverride=&Fort_WeaponRanged_Fire_3P_PSM}
 }
 {
  id=SoundClass_3
  children [Attenuation_2]
  properties {soundClassOverride=&Fort_WeaponRanged_Fire_3P_PSM}
 }
 {
  id=SoundClass_4
  children [Attenuation_3]
  properties {soundClassOverride=&Fort_WeaponRanged_Fire_Distant}
 }
 {id=WavePlayer_15, properties {soundWaveAssetPtr=/Figment_Athena_Dynamite/Audio/Wavs/TNT_Explode_Distant_01}}
 {id=WavePlayer_16, properties {soundWaveAssetPtr=/Figment_Athena_Dynamite/Audio/Wavs/TNT_Explode_Distant_02}}
 {id=WavePlayer_17, properties {soundWaveAssetPtr=/Figment_Athena_Dynamite/Audio/Wavs/TNT_Explode_Distant_03}}
 {id=WavePlayer_18, properties {soundWaveAssetPtr=/Figment_Athena_Dynamite/Audio/Wavs/TNT_Explode_Close_01B}}
 {id=WavePlayer_19, properties {soundWaveAssetPtr=/Figment_Athena_Dynamite/Audio/Wavs/TNT_Explode_Close_02B}}
 {id=WavePlayer_24, properties {soundWaveAssetPtr=&TNT_Explode_Distant_01}}
 {id=WavePlayer_25, properties {soundWaveAssetPtr=&TNT_Explode_Distant_02}}
 {id=WavePlayer_3, properties {soundWaveAssetPtr=&TNT_Explode_Close_01B}}
 {id=WavePlayer_5, properties {soundWaveAssetPtr=&TNT_Explode_Close_02B}}
 {id=WavePlayer_6, properties {soundWaveAssetPtr=/Figment_Athena_Dynamite/Audio/Wavs/TNT_Explode_Close_03B}}
 {id=WavePlayer_7, properties {soundWaveAssetPtr=/Figment_Athena_Dynamite/Audio/Wavs/Dragon_Shotgun_ImpactFuse_03}}
 {id=WavePlayer_8, properties {soundWaveAssetPtr=/Figment_Athena_Dynamite/Audio/Wavs/Dragon_Shotgun_ImpactFuse_02}}
 {id=WavePlayer_9, properties {soundWaveAssetPtr=/Figment_Athena_Dynamite/Audio/Wavs/Dragon_Shotgun_ImpactFuse_01}}
]
properties {bOverrideConcurrency=true, maxDistance=8500, volumeMultiplier=1, concurrencyOverrides {maxCount=2, volumeScale=0.5}}
