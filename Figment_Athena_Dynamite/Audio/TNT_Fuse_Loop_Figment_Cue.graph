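# Sound cue for the looping fuse sound of TNT dynamite in the Figment Athena Dynamite content, played on the explosives sound class.
# A looping fuse wave feeds an Enveloper that fades volume in from silence over a short cubic ramp, then passes through a Modulator to the output.
# It uses a thrown-projectile fuse attenuation asset and overrides concurrency to cap simultaneous fuse loops.
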SoundCue /Figment_Athena_Dynamite/Audio/TNT_Fuse_Loop_Figment_Cue
root=Modulator_1
nodes [
 {
  id=Enveloper_0
  children [WavePlayer_1]
  properties {
   pitchCurve {editorCurveData {keys [{time=0}]}}
   volumeCurve {editorCurveData {keys [{arriveTangent=0.9158, interpMode=RCIM_Cubic, leaveTangent=0.9158, tangentMode=RCTM_User, time=0, value=0}, <at 0.1999 1>]}}
  }
 }
 {id=Modulator_1, children [Enveloper_0]}
 {id=WavePlayer_1, properties {bLooping=true, soundWaveAssetPtr=/Game/Sounds/Fort_GamePlay_Sounds/Explosives/TnT/TNT_Fuse_Loop01}}
]
properties {
 attenuationSettings=/Game/Sounds/Attenuation/Fort_Explosive_Fuse_Prj_Throw_Att
 bOverrideConcurrency=true
 maxDistance=13000
 soundClassObject=/Game/Sounds/FortSoundClassesAndModes/Fort_Explosives
 volumeMultiplier=0.268
 attenuationOverrides {distanceAlgorithm=NaturalSound, falloffDistance=2000, stereoSpread=0}
 concurrencyOverrides {maxCount=3}
}
