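# UI_ItemDrop_Figment_Cue is a sound cue in the Figment_Athena_Dynamite content, named for an item drop sound.
# A single wave player plays UI_DropWeaponImpact01 directly, with no randomization or modulation nodes.
# The cue lowers the volume, uses the shared weapon pickup attenuation and is assigned to the distant ranged weapon fire sound class.
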SoundCue /Figment_Athena_Dynamite/Audio/UI_ItemDrop_Figment_Cue
root=WavePlayer_0
nodes [{id=WavePlayer_0, properties {soundWaveAssetPtr=/Figment_Athena_Dynamite/Audio/Wavs/UI_DropWeaponImpact01}}]
properties {
 attenuationSettings=&Fort_Weapon_Pickup_Att
 maxDistance=3500
 soundClassObject=&Fort_WeaponRanged_Fire_Distant
 volumeMultiplier=0.4
}
